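# Anim Blueprint for the LiquidMetal Mic back bling on the generic mic stand skeleton, adjusting how the backpack sits on the character.
# Two chained Modify Bone nodes act on the `offset` bone over the input pose.
# The first shifts, tilts and shrinks it while skydiving, driven by begin/end skydive events.
# The second adds a nudge for female characters in the front end, with gender cached from the Fort pawn when customization loads.
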AnimBlueprint /SparksCosmetics/Items/Cosmetics/Backpacks/LiquidMetal_Mic/LiquidMetal_Mic_Backpack_AnimBP
target_skeleton=/SparksCharacterCommon/Instrument/MicStand/Generic/Meshes/Generic_MicStand_Skeleton
graphs [
 AnimGraph {nodes [
   n1 = "Transform (Modify) Bone - Bone: offset"(
    ComponentPose=Local To Component(LocalPose=Input Pose())
    Alpha="To Float (Boolean)"(InBool=Get IsSkydiving() -> (IsSkydiving:bool))
    Translation="7,3,-10"
    Rotation="4,-15,4"
    Scale="0.9,0.9,0.9"
   ) {
    rotationMode=BMM_Additive
    rotationSpace=BCS_BoneSpace
    scaleMode=BMM_Replace
    translationMode=BMM_Additive
    translationSpace=BCS_BoneSpace
    alphaBoolBlend {blendOption=Linear}
    boneToModify {boneName=offset}
    rotation=<rot 4 -15 4>
    scale=<0.9 0.9 0.9>
    translation=<7 3 -10>
   }
   n2 = "Transform (Modify) Bone - Bone: offset"!(
    ComponentPose=@n1
    bAlphaBoolEnabled:bool=AND Boolean(A=Get bIsInFrontEnd(), B=Get GenderIsFemale() -> (GenderIsFemale:bool))
    Translation="2,2,0"
    Rotation="0,10,-8"
    Scale="1,1,1"
   ) -> (Pose) {
    alphaInputType=Bool
    rotationMode=BMM_Additive
    rotationSpace=BCS_BoneSpace
    scaleMode=BMM_Replace
    translationMode=BMM_Additive
    translationSpace=BCS_BoneSpace
    alphaBoolBlend {blendOption=Linear}
    boneToModify {boneName=offset}
    rotation {roll=-8, yaw=10}
    translation=<2 2>
   }
   Output Pose(Result=Component To Local(ComponentPose=@n2))
  ]}
 EventGraph {nodes [
   Event OnCharacterCustomizationLoaded()
   n1 = TryGetFortPawn()
   > Is Valid(InputObject=@n1)
   > n2 = FindGender!(execute, Object:Object Reference=@n1) -> (True:Exec, False:Exec, OutGender:EFortCustomGender Enum=Invalid)
   > Set GenderIsFemale(GenderIsFemale:bool="Equal (Enum)"(A:EFortCustomGender Enum=@n2, B:EFortCustomGender Enum=Female)) -> (Output_Get:bool)
   Event OnBeginSkydiving()
   > Set IsSkydiving(IsSkydiving:bool=true) -> (Output_Get:bool)
   Event OnEndSkydiving()
   > Set IsSkydiving(IsSkydiving=false)
   TryGetPawnOwner()
   Event BlueprintUpdateAnimation()
  ]}
]
variables [
 GenderIsFemale {type=boolean}
 IsSkydiving {type=boolean}
]
